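# Material function for the PiteousKicks Pine cape that adds a ghostly glow and animated surface distortion to incoming material attributes.
# Fresnel-weighted, panning swirl noise is tinted, masked and added to emissive, then hue-shifted and contrast-adjusted.
# A scrolling triplanar normal map in pre-skinned space, minus a dominant-axis direction, is masked, converted to tangent space and added to the normal.
# A second output passes through the original normal.
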MaterialFunction /BRCosmetics/Accessories/FORT_Capes/Cape_M_MED_PiteousKicks/Skins/Pine/Materials/MF_PiteousKicks_Pine_Cape_GhostlyGlow
graph {nodes [
  FunctionOutput(None=NamedRerouteUsage()) {outputName=OriginalNormal, sortPriority=1}
  n1 = GetMaterialAttributes(None=FunctionInput() {inputName=MaterialAttributes, inputType=FunctionInput_MaterialAttributes}) {attributeGetTypes [#1, #2]}
  n2 = Multiply(
   A=ComponentMask(None=NamedRerouteUsage()) {r=true}
   B=ScalarParameter() {defaultValue=1, group=PiteousKicks_EmissiveLook, parameterName=EmissiveNoise_MaskHardness, sortPriority=1000}
  )
  n3 = Add(A=@n2, B=ScalarParameter() {defaultValue=0.1, group=PiteousKicks_EmissiveLook, parameterName=EmissiveNoise_MaskGap_01, sortPriority=1000})
  n4 = ScalarParameter() {defaultValue=0.1, group=PiteousKicks_EmissiveLook, parameterName=EmissiveNoise_MaskGap_02, sortPriority=1000}
  n5 = Add(A=OneMinus(None=Saturate(None=@n3)), B=Saturate(None=Subtract(A=@n2, B=@n4)))
  n6 = Power(
   Base=FunctionInput() {inputName=Fresnel, inputType=FunctionInput_Scalar, sortPriority=10}
   Exp=ScalarParameter() {defaultValue=0.2, group=PiteousKicks_EmissiveLook, parameterName=FresnelPower, sortPriority=10}
  )
  n7 = VectorParameter() {group=PiteousKicks_EmissiveLook, parameterName=NoiseSpeed, sortPriority=50, defaultValue=<rgba 0 0.02 0 0>}
  n8 = Panner(Coordinate=TextureCoordinate(), Speed=ComponentMask(None=@n7.RGB) {g=true, r=true}) {speedX=0.1}
  n9 = ScalarParameter() {defaultValue=2, group=PiteousKicks_EmissiveLook, parameterName=NoiseTiling}
  n10 = ComponentMask(None=NamedRerouteUsage()) {g=true, r=true}
  n11 = Constant() {r=0.5}
  n12 = TextureSample(UVs=Add(A=Multiply(A=@n8, B=@n9) {constB=2}, B=Multiply(A=@n10, B=@n11))) {
   samplerSource=SSM_Wrap_WorldGroupSettings
   samplerType=SAMPLERTYPE_LinearColor
   texture=/Game/Global/Materials/MaterialGym/Materials/GhostNoise/Textures/T_ghost_swirls_01
  }
  n13 = VectorParameter() {group=PiteousKicks_EmissiveLook, parameterName=GlowColor, sortPriority=10, defaultValue=<rgba 0 1 0.1261 1>}
  n14 = Multiply(
   A=Multiply(
    A=Multiply(A=@n6, B=@n12.R)
    B=ScalarParameter() {defaultValue=2, group=PiteousKicks_EmissiveLook, parameterName=GlowIntensity, sortPriority=10}
   ) {constB=1.3}
   B=@n13.RGB
  )
  n15 = MaterialFunctionCall(
   Hue Shift Percentage=ScalarParameter() {group=PiteousKicks_EmissiveLook, parameterName=HueShiftPorcentage, sortPriority=18}
   Texture=Add(A=@n1.EmissiveColor, B=Multiply(A=@n5, B=@n14))
  ) {materialFunction=&HueShift}
  n16 = ScalarParameter() {group=PiteousKicks_EmissiveLook, parameterName=EmissiveColorContrast, sortPriority=18}
  FunctionOutput(None=SetMaterialAttributes!(
    MaterialAttributes=@n1.MaterialAttributes
    EmissiveColor=MaterialFunctionCall!(In=@n15.Result, Contrast=@n16) {materialFunction=/Engine/Functions/Engine_MaterialFunctions01/ImageAdjustment/CheapContrast_RGB}.Result
    Normal=NamedRerouteUsage()
   ) {attributeSetTypes [#1, #2]})
  n17 = VertexInterpolator(VS=PreSkinnedPosition())
  NamedRerouteDeclaration(None=@n17.PS) {name=PreskinnedPosition, nodeColor=<rgba 0 0.8359 1 1>}
  n18 = Time()
  n19 = VectorParameter() {group=PiteousKicks_EmissiveLook, parameterName=NormalNoise_Speed, sortPriority=66, defaultValue=<rgba 0 1 -10 1>}
  n20 = Add(
   A=@n17.PS
   B=Multiply(A=@n18, B=MaterialFunctionCall(B=@n19.B) {materialFunction=&AppendMany}.RGB)
  )
  n21 = ScalarParameter() {defaultValue=0.04, group=PiteousKicks_EmissiveLook, parameterName=NormalNoise_Tiling, sortPriority=66}
  n22 = TextureSample(UVs=ComponentMask(None=Multiply(A=@n20, B=@n21) {constB=0.1}) {b=true, g=true}) {
   samplerSource=SSM_Wrap_WorldGroupSettings
   samplerType=SAMPLERTYPE_Normal
   texture=/Game/Characters/Player/Male/Medium/Bodies/M_MED_Cat_Burglar_Ghost/Textures/T_M_MED_Cat_Burglar_Ghost_N1
  }
  n23 = Add(
   A=@n17.PS
   B=Multiply(A=@n18, B=MaterialFunctionCall(R=@n19.R, B=@n19.R) {materialFunction=&AppendMany}.RGB)
  )
  n24 = TextureSample(UVs=ComponentMask(None=Multiply(A=@n23, B=@n21) {constB=0.1}) {b=true, r=true}) {
   samplerSource=SSM_Wrap_WorldGroupSettings
   samplerType=SAMPLERTYPE_Normal
   texture=&T_M_MED_Cat_Burglar_Ghost_N1
  }
  n25 = VertexInterpolator(VS=PreSkinnedNormal())
  n26 = Multiply(A=@n25.PS) {constB=-1}
  n27 = Add(A=Saturate(None=ComponentMask(None=@n25.PS) {g=true}), B=Saturate(None=ComponentMask(None=@n26) {g=true}))
  n28 = Add(
   A=@n17.PS
   B=Multiply(A=@n18, B=MaterialFunctionCall(R=@n19.R) {materialFunction=&AppendMany}.RGB)
  )
  n29 = TextureSample(UVs=ComponentMask(None=Multiply(A=@n28, B=@n21)) {g=true, r=true}) {
   samplerSource=SSM_Wrap_WorldGroupSettings
   samplerType=SAMPLERTYPE_Normal
   texture=&T_M_MED_Cat_Burglar_Ghost_N1
  }
  n30 = Saturate(None=ComponentMask(None=@n25.PS) {b=true})
  n31 = Saturate(None=ComponentMask(None=Multiply(A=@n25.PS) {constB=-1}) {b=true})
  n32 = LinearInterpolate(
   A=LinearInterpolate(A=@n22.RGB, B=@n24.RGB, Alpha=SmoothStep(Value=@n27) {constMin=0.4})
   B=@n29.RGB
   Alpha=SmoothStep(Value=Add(A=@n30, B=@n31)) {constMin=0.4}
  )
  n33 = VectorParameter() {group=PiteousKicks_EmissiveLook, parameterName=NormalNoise_ActorOffset, sortPriority=67, defaultValue=<rgba 0 0 0 1>}
  n34 = Add(A=@n17.PS, B=@n33.RGB)
  n35 = Multiply(A=Constant3Vector() {constant=<rgba 0 1 0 0>}, B=ComponentMask(None=@n34) {g=true})
  n36 = ComponentMask(None=@n34) {r=true}
  n37 = Multiply(A=Constant3Vector() {constant=<rgba 1 0 0 0>}, B=@n36)
  n38 = Constant() {r=10}
  n39 = MaterialFunctionCall(In=Abs(None=@n36), Contrast=@n38) {materialFunction=&CheapContrast}
  n40 = Constant3Vector() {constant=<rgba 0 0 1 0>}
  n41 = ComponentMask(None=@n34) {b=true}
  n42 = Abs(None=@n41)
  n43 = Normalize(VectorInput=LinearInterpolate(
    A=LinearInterpolate(A=@n35, B=@n37, Alpha=@n39.Result)
    B=Multiply(A=@n40, B=@n41)
    Alpha=MaterialFunctionCall(In=@n42, Contrast=@n38) {materialFunction=&CheapContrast}.Result
   ))
  n44 = Multiply(
   A=ComponentMask(None=NamedRerouteUsage()) {r=true}
   B=ScalarParameter() {defaultValue=1, group=PiteousKicks_EmissiveLook, parameterName=NormalNoise_MaskHardness, sortPriority=1000}
  )
  n45 = ScalarParameter() {defaultValue=0.1, group=PiteousKicks_EmissiveLook, parameterName=NormalNoise_MaskGap_01, sortPriority=1000}
  n46 = OneMinus(None=Saturate(None=Add(A=@n44, B=@n45)))
  n47 = Saturate(None=Subtract(A=@n44, B=ScalarParameter() {defaultValue=0.1, group=PiteousKicks_EmissiveLook, parameterName=NormalNoise_MaskGap_02, sortPriority=1000}))
  n48 = Multiply(
   A=Multiply(
    A=ScalarParameter() {defaultValue=1, group=PiteousKicks_EmissiveLook, parameterName=NormalNoise_Intensity, sortPriority=66}
    B=Subtract(A=@n32, B=@n43)
   )
   B=NamedRerouteDeclaration(None=Add(A=@n46, B=@n47)) {name=EffectMask_Normal, nodeColor=<rgba 0 1 0.0156 1>}
  )
  NamedRerouteDeclaration(None=Add(A=NamedRerouteUsage(), B=Transform(None=@n48) {transformSourceType=TRANSFORMSOURCE_Local, transformType=TRANSFORM_Tangent})) {name=Normal, nodeColor=<rgba 1 0 0.2253 1>}
  NamedRerouteDeclaration(None=@n1.Normal) {name=OriginalNormal, nodeColor=<rgba 1 0.6328 0 1>}
  Reroute(None=@n1.Normal)
  Transform(None=VertexNormalWS()) {transformSourceType=TRANSFORMSOURCE_World, transformType=TRANSFORM_Local}
 ]}
