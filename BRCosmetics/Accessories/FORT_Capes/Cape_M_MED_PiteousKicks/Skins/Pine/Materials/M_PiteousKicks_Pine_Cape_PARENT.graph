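# Parent material for the Pine skin of the PiteousKicks cape cosmetic, built on the shared 2023 character material functions with material attributes.
# PreFX output passes through a ghostly glow function, then into a PiteousKicks cape function that also takes the Tech Art Mask.
# A UseReactivity scalar, defaulting to fully on, blends between the glow-only and cape-function attributes before PostFX finishes the material.
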Material /BRCosmetics/Accessories/FORT_Capes/Cape_M_MED_PiteousKicks/Skins/Pine/Materials/M_PiteousKicks_Pine_Cape_PARENT
graph {
 properties {
  bAutomaticallySetUsageInEditor=false
  bUseMaterialAttributes=true
  bUsedWithInstancedStaticMeshes=true
  bUsedWithMeshDeformer=true
  bUsedWithSkeletalMesh=true
  pixelDepthOffsetMode=PDOM_Legacy
  previewMesh=/CosmeticCompanions/Assets/Quadruped/CosmoHound_Meteor/Meshes/CosmoHound_Meteor_Quadruped
  shadingModel=MSM_FromMaterialExpression
  usedShadingModels=Default Lit | Subsurface | Clear Coat | Subsurface Profile | Cloth
 }
 nodes [
  n1 = MaterialFunctionCall() {materialFunction=&MF_CharacterParent_2023_PreFX}
  n2 = MaterialFunctionCall!(MaterialAttributes=@n1.Result, Fresnel=@n1.Result) {materialFunction=&/MF_PiteousKicks_Pine_Cape_GhostlyGlow}
  MaterialFunctionCall(
   Material=BlendMaterialAttributes(
    A=@n2.Result
    B=MaterialFunctionCall!(Tech Art Mask=@n1.Tech Art Mask, MaterialAttributes=@n2.Result, OriginalNormal=@n2.Result) {materialFunction=/Game/Accessories/FORT_Capes/Cape_M_MED_PiteousKicks/Skins/Pine/Materials/MF_PiteousKicks_Cape}.Result
    Alpha=ScalarParameter() {defaultValue=1, group=PiteousKicks_GoldMaterial, parameterName=UseReactivity}
   )
   M=@n1.M
   Pre Skinned Local Position=@n1.M
   Distance Blend=@n1.M
   Fresnel=@n1.M
  ) {materialFunction=&MF_CharacterParent_2023_PostFX}
 ]
}
